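# Sound cue for the VR Editor UI that plays a button click sound.
# A single wave player node is the root and feeds the Click_on_Button sound wave straight to output, with no randomization, modulation or mixing.
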SoundCue /Engine/VREditor/Sounds/UI/Click_on_Button_Cue
root=WavePlayer_0
nodes [{id=WavePlayer_0, properties {soundWaveAssetPtr=&/Click_on_Button}}]
properties {maxDistance=2097152}
